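# Sound cue for a dock spot ambience of water hitting a wooden pontoon.
# An equally weighted Random node picks one of six impact waves, feeding a Modulator for slight volume variation.
# An Enveloper at the root applies an endlessly looping, subtle volume dip; no Looping node repeats the waves.
# It uses a distant-creak one-shot attenuation asset, an indoor ambient sound class and, by its name, a two-voice prevent-new concurrency.
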SoundCue /WildEstate/Assets/Audio/POI/Dock/Spots/SC_Spot_Water_Impact_On_Wood_Pontoon
root=Enveloper_0
nodes [
 {
  id=Enveloper_0
  children [Modulator_0]
  properties {
   bLoop=true
   bLoopIndefinitely=true
   loopEnd=20
   pitchCurve {editorCurveData {keys [{time=0}]}}
   volumeCurve {editorCurveData {keys [<at -0.0028 1>,<at 0 1>,<at 10 0.95>,<at 20 1>]}}
  }
 }
 {id=Modulator_0, children [Random_0], properties {volumeMax=1, volumeMin=0.9}}
 {
  id=Random_0
  children [WavePlayer_7, WavePlayer_8, WavePlayer_9, WavePlayer_10, WavePlayer_11, WavePlayer_12]
  properties {weights [1,1,1,1,1,1]}
 }
 {id=WavePlayer_10, properties {soundWaveAssetPtr=/Game/Athena/Sounds/Water/Ambient/Spot_Water_Pontoon_Wood_Impact_04}}
 {id=WavePlayer_11, properties {soundWaveAssetPtr=/Game/Athena/Sounds/Water/Ambient/Spot_Water_Pontoon_Wood_Impact_05}}
 {id=WavePlayer_12, properties {soundWaveAssetPtr=/Game/Athena/Sounds/Water/Ambient/Spot_Water_Pontoon_Wood_Impact_06}}
 {id=WavePlayer_7, properties {soundWaveAssetPtr=/Game/Athena/Sounds/Water/Ambient/Spot_Water_Pontoon_Wood_Impact_01}}
 {id=WavePlayer_8, properties {soundWaveAssetPtr=/Game/Athena/Sounds/Water/Ambient/Spot_Water_Pontoon_Wood_Impact_02}}
 {id=WavePlayer_9, properties {soundWaveAssetPtr=/Game/Athena/Sounds/Water/Ambient/Spot_Water_Pontoon_Wood_Impact_03}}
]
properties {
 attenuationSettings=/Game/Sounds/Attenuation/Fort_TileAmbient_OneShot_Distant_Creak_Att
 bDebug=true
 maxDistance=8200
 soundClassObject=&Fort_Ambient_Indoors_Athena
 volumeMultiplier=1
 attenuationOverrides {bSpatialize=false, stereoSpread=1147.9967, attenuationShapeExtents {x=6826.1992}}
 concurrencySet [/AmbientAudioBase/Concurrency/AmbientConc_2Max_Loop_PreventNew_Default_Concurrency]
}
